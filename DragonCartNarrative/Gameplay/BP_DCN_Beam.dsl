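; Tractor beam actor for DragonCart narrative that lifts players, repels vehicles and destroys tagged TechHQ window covers.
; On the server, the outer radius toggles a beam FX effect, while the inner radius puts on-foot pawns in a vortex with fall immunity, tractor-beam and optional damage-over-time effects.
; Non-pawn actors go to KnockbackVehicle, which disables vehicles, counters boost and applies an outward impulse.
; A separately triggered timer re-checks overlapping pawns.
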(event OnComponentBeginOverlap(OuterRadius)
  (params (OverlappedComponent :type Primitive%20Component%20Object%20Reference) (OtherActor :type Actor%20Object%20Reference) (OtherComp :type Primitive%20Component%20Object%20Reference) (OtherBodyIndex :type Integer) (bFromSweep :type Boolean) (SweepResult :type Hit%20Result%20Structure%20%28by%20ref%29))
  (Networking|SwitchHasAuthority
    (:Authority
      (bind _asfort_player_pawn_athena (Utilities|Casting|CastToFortPlayerPawnAthena OtherActor))
      (Development|PrintString "Player pawn has overlapped Outter Beam")
      (Utilities|IsValid (Ability|GetFortAbilitySystemComponent _asfort_player_pawn_athena)
        (:Is%20Valid
          (GameplayEffects|ApplyGameplayEffectToTarget :self (Ability|GetFortAbilitySystemComponent _asfort_player_pawn_athena) :GameplayEffectClass "/DragonCartNarrative/Gameplay/Effect/GE_DCN_BeamFX.GE_DCN_BeamFX_C" :Target (Ability|GetFortAbilitySystemComponent _asfort_player_pawn_athena) :Level 0.0))))))

(event OnComponentEndOverlap(OuterRadius)
  (params (OverlappedComponent :type Primitive%20Component%20Object%20Reference) (OtherActor :type Actor%20Object%20Reference) (OtherComp :type Primitive%20Component%20Object%20Reference) (OtherBodyIndex :type Integer))
  (Networking|SwitchHasAuthority
    (:Authority
      (Variables|Default|SetOther%20Actor OtherActor)
      (bind _asfort_player_pawn_athena (Utilities|Casting|CastToFortPlayerPawnAthena (Variables|Default|GetOther%20Actor)))
      (Utilities|IsValid (Ability|GetFortAbilitySystemComponent _asfort_player_pawn_athena)
        (:Is%20Valid
          (GameplayEffects|RemoveActiveGameplayEffectbySourceEffect (Ability|GetFortAbilitySystemComponent _asfort_player_pawn_athena) "/DragonCartNarrative/Gameplay/Effect/GE_DCN_BeamFX.GE_DCN_BeamFX_C" 0 -1))))))

(event EventBeginPlay
  (CallFunction|DestroyGlass))

(event EventTick
  (params (DeltaSeconds :type Float%20%28single-precision%29)))

(event Custom|DestroyGlass
  (bind _outactors (Actor|GetAllActorsOfClasswithTag "/Hera/Sets/FriedNexus/Blueprints/Gateway/BP_FriedNexus_Gateway_HoleCover_A_A.BP_FriedNexus_Gateway_HoleCover_A_A_C" "TechHQWindow"))
  (for _array_element _outactors
    (Actor|DestroyActor _array_element)))

(event Custom|VortexOverlappingPawns
  (bind _returnvalue (Utilities|Time|SetTimerbyEvent (delegate Custom|CheckOverlaps) 0.5 true))
  (Variables|Default|SetOverlapTimerHandle _returnvalue))

(event Custom|CheckOverlaps
  (for _array_element (Collision|GetOverlappingActors (Variables|Default|GetInnerRadius) "/Script/FortniteGame.FortPlayerPawn" :!class /Script/Engine.PrimitiveComponent)
    (Networking|SwitchHasAuthority
      (:Authority
        (bind _output_get (Variables|Default|SetOther%20Actor _array_element))
        (bind _returnvalue (GameplayTags|HasMatchingGameplayTag(Message) _output_get "(TagName=\"Gameplay.Status.TractorBeam\")"))
        (if (not _returnvalue)
          (bind _asfort_player_pawn_athena (Utilities|Casting|CastToFortPlayerPawnAthena (Variables|Default|GetOther%20Actor))
            (:then
              (CallFunction|DestroyGlass)
              (if (not (or (Vehicle|IsDrivingVehicle _asfort_player_pawn_athena) (Vehicle|IsPassengerinVehicle _asfort_player_pawn_athena)))
                (Tags|CancelAbilitiesWithTags (Ability|GetFortAbilitySystemComponent _asfort_player_pawn_athena) (Variables|Default|GetWingsuitTag))
                (Development|PrintString "Player pawn has overlapped Inner Beam and will be put in vortex")
                (Vortex|SetinVortex _asfort_player_pawn_athena true 5e+01 5e+02 5e+03 -1.0 0.0 false)
                (Utilities|IsValid (Ability|GetFortAbilitySystemComponent _asfort_player_pawn_athena)
                  (:Is%20Valid
                    (GameplayEffects|FortApplyGameplayEffectToSelf :self (Ability|GetFortAbilitySystemComponent _asfort_player_pawn_athena) :GameplayEffectClass "/Game/Athena/Items/Parents/FallImmunity/GE_Athena_TillLandFallDamageImmunity_NoFX_Grant_ApplyInAirOnly.GE_Athena_TillLandFallDamageImmunity_NoFX_Grant_ApplyInAirOnly_C" :Level 0.0)
                    (GameplayEffects|FortApplyGameplayEffectToSelf :self (Ability|GetFortAbilitySystemComponent _asfort_player_pawn_athena) :GameplayEffectClass (Variables|Default|GetTractorBeamEffect) :Level 0.0)
                    (if (>= (ScalableFloat|GetValueatLevel (Variables|Default|GetDOTEnabled)) 1.0)
                      (GameplayEffects|FortApplyGameplayEffectToSelf :self (Ability|GetFortAbilitySystemComponent _asfort_player_pawn_athena) :GameplayEffectClass (Variables|Default|GetTractorBeamDOT) :OptionalDamageSource self :Level 0.0))))))
            (:CastFailed
              (CallFunction|KnockbackVehicle (Variables|Default|GetOther%20Actor)))))))))

(event OnComponentBeginOverlap(InnerRadius)
  (params (OverlappedComponent :type Primitive%20Component%20Object%20Reference) (OtherActor :type Actor%20Object%20Reference) (OtherComp :type Primitive%20Component%20Object%20Reference) (OtherBodyIndex :type Integer) (bFromSweep :type Boolean) (SweepResult :type Hit%20Result%20Structure%20%28by%20ref%29))
  (Networking|SwitchHasAuthority
    (:Authority
      (Variables|Default|SetOther%20Actor OtherActor)
      (bind _asfort_player_pawn_athena (Utilities|Casting|CastToFortPlayerPawnAthena (Variables|Default|GetOther%20Actor))
        (:then
          (CallFunction|DestroyGlass)
          (if (not (or (Vehicle|IsDrivingVehicle _asfort_player_pawn_athena) (Vehicle|IsPassengerinVehicle _asfort_player_pawn_athena)))
            (Tags|CancelAbilitiesWithTags (Ability|GetFortAbilitySystemComponent _asfort_player_pawn_athena) (Variables|Default|GetWingsuitTag))
            (Development|PrintString "Player pawn has overlapped Inner Beam and will be put in vortex")
            (Vortex|SetinVortex _asfort_player_pawn_athena true 5e+01 5e+02 5e+03 -1.0 0.0 false)
            (Utilities|IsValid (Ability|GetFortAbilitySystemComponent _asfort_player_pawn_athena)
              (:Is%20Valid
                (GameplayEffects|FortApplyGameplayEffectToSelf :self (Ability|GetFortAbilitySystemComponent _asfort_player_pawn_athena) :GameplayEffectClass "/Game/Athena/Items/Parents/FallImmunity/GE_Athena_TillLandFallDamageImmunity_NoFX_Grant_ApplyInAirOnly.GE_Athena_TillLandFallDamageImmunity_NoFX_Grant_ApplyInAirOnly_C" :Level 0.0)
                (GameplayEffects|FortApplyGameplayEffectToSelf :self (Ability|GetFortAbilitySystemComponent _asfort_player_pawn_athena) :GameplayEffectClass (Variables|Default|GetTractorBeamEffect) :Level 0.0)
                (if (>= (ScalableFloat|GetValueatLevel (Variables|Default|GetDOTEnabled)) 1.0)
                  (GameplayEffects|FortApplyGameplayEffectToSelf :self (Ability|GetFortAbilitySystemComponent _asfort_player_pawn_athena) :GameplayEffectClass (Variables|Default|GetTractorBeamDOT) :OptionalDamageSource self :Level 0.0))))))
        (:CastFailed
          (CallFunction|KnockbackVehicle (Variables|Default|GetOther%20Actor)))))))

(event OnComponentEndOverlap(InnerRadius)
  (params (OverlappedComponent :type Primitive%20Component%20Object%20Reference) (OtherActor :type Actor%20Object%20Reference) (OtherComp :type Primitive%20Component%20Object%20Reference) (OtherBodyIndex :type Integer))
  (Networking|SwitchHasAuthority
    (:Authority
      (bind _output_get (Variables|Default|SetOther%20Actor OtherActor))
      (bind _returnvalue (GameplayTags|HasMatchingGameplayTag(Message) _output_get "(TagName=\"Gameplay.Status.Knockback\")"))
      (if (not _returnvalue)
        (Utilities|FlowControl|Delay 0.1))
      (bind _asfort_player_pawn_athena (Utilities|Casting|CastToFortPlayerPawnAthena (Variables|Default|GetOther%20Actor)))
      (Vortex|SetinVortex _asfort_player_pawn_athena false 0.0 0.0 0.0 0.0 0.0 false)
      (Utilities|IsValid (Ability|GetFortAbilitySystemComponent _asfort_player_pawn_athena)
        (:Is%20Valid
          (GameplayEffects|RemoveActiveGameplayEffectbySourceEffect (Ability|GetFortAbilitySystemComponent _asfort_player_pawn_athena) (Variables|Default|GetTractorBeamEffect) 0 -1)
          (GameplayEffects|RemoveActiveGameplayEffectbySourceEffect (Ability|GetFortAbilitySystemComponent _asfort_player_pawn_athena) (Variables|Default|GetTractorBeamDOT) 0 -1))))))

(fn ApplyTeleportGA
  (params (PlayerToTeleport :type Fort%20Player%20Pawn%20Athena%20Object%20Reference))
  (bind _returnvalue (Ability|GetFortAbilitySystemComponent PlayerToTeleport))
  (Utilities|IsValid _returnvalue
    (:Is%20Valid
      (GameplayEffects|FortApplyGameplayEffectToTarget :self _returnvalue :GameplayEffectClass (Variables|Default|GetTeleportEffect) :TargetOfEffect _returnvalue :OptionalSourceObject self :OptionalEffectCauser self :Level 0.0))))

(fn GetVehicleImpulseValue
  (params (Actor :type Actor%20Object%20Reference))
  (bind (_returnvalue_x _returnvalue_y) (Math|Vector|BreakVector (- (Transformation|GetActorLocation Actor) (Transformation|GetWorldLocation (Variables|Default|GetInnerRadius)))))
  (return (+ (* (Math|Vector|Normalize (Math|Vector|MakeVector _returnvalue_x _returnvalue_y)) (Math|Vector|MakeVector (ScalableFloat|GetValueatLevelwithContext (Variables|Default|GetVehicle%20Knockback%20-%20Horz)) (ScalableFloat|GetValueatLevelwithContext (Variables|Default|GetVehicle%20Knockback%20-%20Horz)))) (Math|Vector|MakeVector 0.0 0.0 (ScalableFloat|GetValueatLevelwithContext (Variables|Default|GetVehicle%20Knockback%20-%20Vert))))))

(fn KnockbackVehicle
  (params (Object :type Actor%20Object%20Reference))
  (bind (_returnvalue_x _returnvalue_y) (Math|Vector|BreakVector (Transformation|GetWorldLocation (Variables|Default|GetInnerRadius))))
  (Networking|SwitchHasAuthority
    (:Authority
      (bind _asfort_vehicle_interface (Utilities|Casting|CastToFortVehicleInterface Object))
      (Vehicle|WakeUp _asfort_vehicle_interface)
      (GameplayEffects|ApplyGameplayEffectToTarget :self (Vehicle|GetVehicleAbilitySystemComponent _asfort_vehicle_interface) :GameplayEffectClass (Variables|Default|GetVehicleDisableEffect) :Target (Vehicle|GetVehicleAbilitySystemComponent _asfort_vehicle_interface) :Level 0.0)
      (if (Class|FortVehicleInterface|IsBoosting _asfort_vehicle_interface)
        (bind _returnvalue (Vehicle|Cosmetics|GetBodySkeletalMeshComponent(Message) Object))
        (bind (_returnvalue_x_2 _returnvalue_y_2 _returnvalue_z) (Math|Vector|BreakVector (Transformation|GetWorldLocation _returnvalue)))
        (Physics|AddRadialForce _returnvalue (Math|Vector|MakeVector _returnvalue_x _returnvalue_y _returnvalue_z) 8e+02 (ScalableFloat|GetValueatLevelwithContext (Variables|Default|GetVehicle%20Knockback%20Radial%20Force)) "RIF_Constant" true) ; Apply Force to cancel out boost.
      )
      (Audio|BroadcastAbilityImpactatLocation 0 (Utilities|ToObject(Interface) _asfort_vehicle_interface) (Default|GetVehicleImpulseValue Object) (ScalableFloat|GetValueatLevelwithContext (Variables|Default|GetVehicle%20Knockback%20Impulse%20Radius))))))

(fn OnRep_TargetRift)

(fn ConstructionScript)
