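; DCN shard teleporter actor that sends players entering its sphere to rift locations and repels vehicles.
; On server overlap, it ejects vehicle passengers for teleport via gameplay effect, then knocks the vehicle back, countering boost and retrying on a timer.
; Targets are random off-cooldown BP_DCN_TeleportLocation actors with a safe-zone check, falling back to "Failsafe"-tagged points.
; A hotfix scalable float and ST_Active/ST_Inactive events toggle sphere collision.
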(event EventBeginPlay
  (bind _hf_teleportenabled (Variables|Default|GetHF_TeleportEnabled))
  (bind _returnvalue (ScalableFloat|GetValueasBoolwithContext _hf_teleportenabled))
  (Development|PrintString "I have awoken!" true true "(R=0.000000,G=0.660000,B=1.000000,A=1.000000)" 2e+01)
  (if (not _returnvalue)
    (Development|PrintString (Utilities|String|Append "I am enabled: " (Utilities|String|ToString(Boolean) _returnvalue)) true true "(R=0.000000,G=0.660000,B=1.000000,A=1.000000)" 2e+01)
    (Collision|SetCollisionEnabled (Variables|Default|GetTeleportSphere) "NoCollision")))

(event OnComponentBeginOverlap(TeleportSphere)
  (params (OverlappedComponent :type Primitive%20Component%20Object%20Reference) (OtherActor :type Actor%20Object%20Reference) (OtherComp :type Primitive%20Component%20Object%20Reference) (OtherBodyIndex :type Integer) (bFromSweep :type Boolean) (SweepResult :type Hit%20Result%20Structure%20%28by%20ref%29))
  (Networking|SwitchHasAuthority
    (:Authority
      (if (Utilities|Array|IsEmpty (Variables|Default|GetTeleportLocations))
        (bind _outactors (Actor|GetAllActorsOfClass "/DragonCartNarrative/Gameplay/BP_DCN_TeleportLocation.BP_DCN_TeleportLocation_C"))
        (Variables|Default|SetTeleportLocations _outactors))
      (bind _asfort_athena_vehicle (Utilities|Casting|CastToFortAthenaVehicle OtherActor)
        (:then
          (CallFunction|TeleportVehicleOccupants _asfort_athena_vehicle)
          (CallFunction|KnockbackVehicle _asfort_athena_vehicle))
        (:CastFailed
          (bind _asfort_player_pawn_athena (Utilities|Casting|CastToFortPlayerPawnAthena OtherActor))
          (CallFunction|TeleportPawns _asfort_player_pawn_athena))))))

(event OnComponentEndOverlap(TeleportSphere)
  (params (OverlappedComponent :type Primitive%20Component%20Object%20Reference) (OtherActor :type Actor%20Object%20Reference) (OtherComp :type Primitive%20Component%20Object%20Reference) (OtherBodyIndex :type Integer)))

(event Custom|ST_Active
  (Collision|SetCollisionEnabled (Variables|Default|GetTeleportSphere) "QueryOnly"))

(event Custom|ST_Inactive
  (Collision|SetCollisionEnabled (Variables|Default|GetTeleportSphere) "NoCollision"))

(event Custom|ST_KnockbackCar
  (bind _overlapping_actors (Variables|Default|GetOverlapping%20Actors))
  (Variables|Default|SetOverlapping%20Actors (Collision|GetOverlappingActors (Variables|Default|GetTeleportSphere) "/Script/FortniteGame.FortAthenaVehicle" :!class /Script/Engine.PrimitiveComponent))
  (if (> (Utilities|Array|Length _overlapping_actors) 0)
    (for _array_element _overlapping_actors
      (bind _asfort_athena_vehicle (Utilities|Casting|CastToFortAthenaVehicle _array_element))
      (CallFunction|KnockbackVehicle _asfort_athena_vehicle))))

(event Custom|RetryVehicleKick
  (bind _overlapping_actors (Variables|Default|GetOverlapping%20Actors))
  (Utilities|Time|ClearandInvalidateTimerbyHandle (Variables|Default|GetRekickTimerHandle))
  (Variables|Default|SetOverlapping%20Actors (Collision|GetOverlappingActors (Variables|Default|GetTeleportSphere) "/Script/FortniteGame.FortAthenaVehicle" :!class /Script/Engine.PrimitiveComponent))
  (if (> (Utilities|Array|Length _overlapping_actors) 0)
    (for _array_element _overlapping_actors
      (bind _asfort_athena_vehicle (Utilities|Casting|CastToFortAthenaVehicle _array_element))
      (CallFunction|KnockbackVehicle _asfort_athena_vehicle))))

(fn ApplyTeleportGA
  (params (PlayerToTeleport :type Fort%20Player%20Pawn%20Athena%20Object%20Reference))
  (bind _returnvalue (Ability|GetFortAbilitySystemComponent PlayerToTeleport))
  (Utilities|IsValid _returnvalue
    (:Is%20Valid
      (GameplayEffects|FortApplyGameplayEffectToTarget :self _returnvalue :GameplayEffectClass (Variables|Default|GetTeleportEffect) :TargetOfEffect _returnvalue :OptionalSourceObject self :OptionalEffectCauser self :Level 0.0))
    (:Is%20Not%20Valid
      (Fortnite|FortShippingLog self "DCN_Beam: Player not VALIDDD for teleport!!!!"))))

(fn GetImpulseValue
  (params (OtherActor :type Actor%20Object%20Reference))
  (bind (_returnvalue_x _returnvalue_y) (Math|Vector|BreakVector (- (Transformation|GetActorLocation OtherActor) (Transformation|GetWorldLocation (Variables|Default|GetTeleportSphere)))))
  ;[0 0.15 0.08295 0.5] Caluclate knockback impulse
  (return (+ (* (Math|Vector|Normalize (Math|Vector|MakeVector _returnvalue_x _returnvalue_y)) (Math|Vector|MakeVector (LocalVariables|GetVehicle%20Knockback%20-%20Horz) (LocalVariables|GetVehicle%20Knockback%20-%20Horz))) (Math|Vector|MakeVector 0.0 0.0 (LocalVariables|GetVehicle%20Knockback%20-%20Vert)))))

(fn KnockbackVehicle
  (params (Vehicle :type Fort%20Athena%20Vehicle%20Object%20Reference))
  (Networking|SwitchHasAuthority
    (:Authority
      (Vehicle|WakeUp Vehicle)
      (Components|Movement|StopMovementImmediately (Pawn|GetMovementComponent Vehicle))
      (if (Class|FortVehicleInterface|IsBoosting Vehicle)
        (Physics|AddRadialForce (Variables|/Script/FortniteGame.FortAthenaVehicle|GetSkeletalMesh Vehicle) (Transformation|GetWorldLocation (Variables|Default|GetTeleportSphere)) 777.0 2e+02 "RIF_Constant" true) ; Apply Force to cancel out boost.
      )
      (Audio|BroadcastAbilityImpactatLocation 0 Vehicle (CallFunction|GetImpulseValue Vehicle) 777.0)
      (Development|PrintString (Utilities|String|Append "I knocked back: " (Utilities|GetDisplayName Vehicle)))
      (bind _returnvalue (Utilities|Time|SetTimerbyEvent (EventDispatchers|CreateEvent :!event RetryVehicleKick) 0.25))
      (Variables|Default|SetRekickTimerHandle _returnvalue))))

(fn OnRep_ActorOverlapLocation
  (bind _actor_overlap_location (Variables|Default|GetActor%20Overlap%20Location))
  (Development|PrintString (Utilities|String|Append "I should be spawning a niagara at: " (Utilities|String|ToString(Vector) _actor_overlap_location)))
  (Networking|CanExecuteCosmeticEvents
    (:True
      (Niagara|SpawnSystematLocation :SystemTemplate "/Game/Athena/Items/ForagedItems/Rift/FX/NS_WildRift_Portal_Impact.NS_WildRift_Portal_Impact" :Location _actor_overlap_location :Scale "1.000000,1.000000,1.000000" :bAutoDestroy false :bAutoActivate true :PoolingMethod "None" :bPreCullCheck true))))

(fn SelectTeleportTarget
  (bind _pointsoffcooldown (LocalVariables|GetPointsOffCooldown))
  (bind _outitem (Utilities|Array|RandomArrayItem _pointsoffcooldown))
  (bind _returnvalue (Utilities|Array|Length _pointsoffcooldown))
  (bind _returnvalue_2 (< _returnvalue 1))
  (Utilities|Array|Clear (LocalVariables|GetPointsOffCooldown))
  (Utilities|Array|Shuffle (Variables|Default|GetTeleportLocations))
  (for _array_element (Variables|Default|GetTeleportLocations)
    (if (not (Variables|/DragonCartNarrative/Gameplay/BP_DCN_TeleportLocation.BP_DCN_TeleportLocation_C|GetCooldown _array_element))
      (Utilities|Array|Add (LocalVariables|GetPointsOffCooldown) _array_element)))
  (if (not _returnvalue_2)
    (bind _returnvalue_3 (SafeZone|IsLocationinSafeZone (Transformation|GetWorldLocation (Variables|/DragonCartNarrative/Gameplay/BP_DCN_TeleportLocation.BP_DCN_TeleportLocation_C|GetDefaultSceneRoot _outitem))))
    (if (not _returnvalue_3)
      (Utilities|Array|RemoveItem (LocalVariables|GetPointsOffCooldown) _outitem)))
  (Variables|Default|SetTeleportData (Utilities|Struct|MakeDCNTeleportStruct _outitem _returnvalue_2)))

(fn TeleportPawns
  (params (TeleportingPawn :type Fort%20Player%20Pawn%20Athena%20Object%20Reference))
  (bind (_teleportdata_foundtarget_4_aa7d7797447ed03a89d16c9128dcc185 _teleportdata_alloncooldown_1_9a8f6eb94a1a49c4a2b18c84b7ea0184) (Utilities|Struct|BreakDCNTeleportStruct (Variables|GetTeleportData_FoundTarget_4_AA7D7797447ED03A89D16C9128DCC185)))
  (CallFunction|SelectTeleportTarget)
  (if (not _teleportdata_alloncooldown_1_9a8f6eb94a1a49c4a2b18c84b7ea0184)
    (Variables|Default|SetTargetRift _teleportdata_foundtarget_4_aa7d7797447ed03a89d16c9128dcc185)
    (else
      (Fortnite|FortShippingLog self "ShardTeleport: All teleport points on cooldown!")
      (bind _outactors (Actor|GetAllActorsOfClasswithTag "/DragonCartNarrative/Gameplay/BP_DCN_TeleportLocation.BP_DCN_TeleportLocation_C" "Failsafe"))
      (bind _asbp_dcn_teleport_location (Utilities|Casting|CastToBP_DCN_TeleportLocation (Utilities|Array|Get(acopy) _outactors)))
      (Variables|Default|SetTargetRift _asbp_dcn_teleport_location)))
  (Utilities|IsValid (Variables|Default|GetTargetRift)
    (:Is%20Valid
      (CallFunction|ApplyTeleportGA TeleportingPawn)
      (Ability|SendGameplayEventtoActor TeleportingPawn "(TagName=\"Abilities.Generic.Rift.Trigger\")")
      (if (Variables|Default|GetUseSkydive)
        (Mission|Skydiving|Skydive true TeleportingPawn)))
    (:Is%20Not%20Valid
      (Fortnite|FortShippingLog self "DCN_Beam: Target Rift is Invalid!!"))))

(fn TeleportVehicleOccupants
  (params (Vehicle :type Fort%20Athena%20Vehicle%20Object%20Reference))
  (bind _teleportdata (Variables|Default|GetTeleportData))
  (bind (_foundtarget_4_aa7d7797447ed03a89d16c9128dcc185 _alloncooldown_1_9a8f6eb94a1a49c4a2b18c84b7ea0184) (Utilities|Struct|BreakDCNTeleportStruct _teleportdata))
  (bind _asfort_vehicle_interface (Utilities|Casting|CastToFortVehicleInterface Vehicle))
  (Development|PrintString "I have the interface" true true "(R=0.000000,G=0.660000,B=1.000000,A=1.000000)" 3e+01)
  (Variables|Default|SetPlayersToTeleport (Vehicle|GetAllPassengers _asfort_vehicle_interface))
  (Vehicle|EjectAllPlayers _asfort_vehicle_interface)
  (for _array_element (Variables|Default|GetPlayersToTeleport)
    (CallFunction|SelectTeleportTarget)
    (if (not _alloncooldown_1_9a8f6eb94a1a49c4a2b18c84b7ea0184)
      (Variables|Default|SetTargetRift _foundtarget_4_aa7d7797447ed03a89d16c9128dcc185)
      (else
        (Fortnite|FortShippingLog self "ShardTeleport: All teleport points on cooldown! Using Failsafe teleport")
        (bind _outactors (Actor|GetAllActorsOfClasswithTag "/DragonCartNarrative/Gameplay/BP_DCN_TeleportLocation.BP_DCN_TeleportLocation_C" "Failsafe"))
        (bind _asbp_dcn_teleport_location (Utilities|Casting|CastToBP_DCN_TeleportLocation (Utilities|Array|Get(acopy) _outactors)))
        (Variables|Default|SetTargetRift _asbp_dcn_teleport_location)))
    (Utilities|IsValid (Variables|Default|GetTargetRift)
      (:Is%20Valid
        (bind _asfort_player_pawn_athena (Utilities|Casting|CastToFortPlayerPawnAthena _array_element))
        (CallFunction|ApplyTeleportGA _asfort_player_pawn_athena))
      (:Is%20Not%20Valid
        (Fortnite|FortShippingLog self "DCN_Beam: Target Rift is Invalid!!")))))

(fn ConstructionScript)
