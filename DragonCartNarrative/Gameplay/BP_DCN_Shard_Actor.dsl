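; Narrative shard actor that moves its static mesh between preset spline points as the OrbTech HQ phase changes.
; OrbTechHQPhase waits a tick, then maps phases 5 to 7 to the first three spline points.
; SetShardLocation teleports the mesh to that point's world position and prints it for debugging.
; BeginPlay waits for playlist and calendar sync, then checks named events, but the active-event branch does nothing.
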(event Custom|OrbTechHQPhase
  (params (Phase :type Integer))
  (Utilities|FlowControl|DelayUntilNextTick)
  (switch int (- Phase 1)
    (:4
      (CallFunction|SetShardLocation 0))
    (:5
      (CallFunction|SetShardLocation 1))
    (:6
      (CallFunction|SetShardLocation 2))))

(event EventBeginPlay
  (|CallParentFunction)
  (Class|FortAsyncActionCurrentPlaylistReady|PlaylistDataReadyAsync
    (:OnReady
      (MCPContext|InitialCalendarSyncCompleted
        (:OnSyncComplete
          (bind _array_element (Utilities|Array|ForEachLoopwithBreak (Variables|Default|GetEventNames))
            (:LoopBody
              (bind _biseventactive (CalendarEvents|GetCalendarEventInformation _array_element))
              (if _biseventactive
                _))))))))

(fn SetShardLocation
  (params (PointIndex :type Integer))
  (bind _spline (Variables|Default|GetSpline))
  (bind (_returnvalue_inputkey _returnvalue_position) (Utilities|Struct|BreakSplinePoint (Spline|GetSplinePointAt _spline PointIndex "World")))
  (Transformation|SetWorldLocation (Variables|Default|GetStaticMesh) _returnvalue_position false true)
  (Development|PrintString (Utilities|String|ToString(Vector) _returnvalue_position)))

(fn ConstructionScript
  (|CallParentFunction))
