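; Teleport location actor for rift shard teleports, with an optional encounter spawn and a flash effect.
; RiftFlash runs on authority: it sets Cooldown, spawns ShardEncounterClass at the actor's transform if SpawnPropsOnShardTeleport is set, then starts a timer that clears Cooldown.
; OnRep_PlayFlash activates RiftEffect where cosmetic events can run, with debug prints on both branches.
; Nothing in this file sets PlayFlash or checks Cooldown.
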(event EventBeginPlay)

(event Custom|RiftFlash
  (Networking|SwitchHasAuthority
    (:Authority
      (Variables|Default|SetCooldown true)
      (if (Variables|Default|GetSpawnPropsOnShardTeleport)
        (Game|SpawnActorfromClass (Variables|Default|GetShardEncounterClass) (Transformation|GetActorTransform) "Undefined" "MultiplyWithRoot" 0 0 0)
        (else
          _))
      (Utilities|Time|SetTimerbyEvent (EventDispatchers|CreateEvent :!event Custom|CooldownTime) (Variables|Default|GetCooldownTimer)))))

(event Custom|CooldownTime
  (Variables|Default|SetCooldown false))

(fn OnRep_PlayFlash
  (Networking|CanExecuteCosmeticEvents
    (:True
      (Development|PrintString "Trigger Flash Client" true true "(R=0.000000,G=0.660000,B=1.000000,A=1.000000)" 2e+01)
      (Components|Activation|Activate (Variables|Default|GetRiftEffect) false))
    (:False
      (Development|PrintString "Trigger Flash Server"))))

(fn ConstructionScript)
