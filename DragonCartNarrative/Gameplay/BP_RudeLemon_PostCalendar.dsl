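; BP_RudeLemon_PostCalendar switches World Partition data layers for the RudeLemon rocket-landing aftermath and two Red Rift days, based on calendar events.
; On server BeginPlay, it waits for playlist data and initial calendar sync, then checks RL_PostEvent, RL_Rift1 and RL_Rift2.
; RL_NextMatch and RedRiftDay set hardcoded pre/post-landing and per-day layer states.
; Calendar-layer functions also toggle list-driven layers per event; CanUpdateInMatch rejects mid-match updates for performance.
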(event EventBeginPlay
  ; RudeLemon / Rocket Landing
  ;;[0.281 0.3165 0.6823 0.5] Begin Play
  (if (Networking|IsServer)
    (Default|LogIt "Begin Play - server")
    (Class|FortAsyncActionCurrentPlaylistReady|PlaylistDataReadyAsync
      (:OnReady
        (MCPContext|InitialCalendarSyncCompleted
          (:OnSyncComplete
            (Utilities|FlowControl|Sequence
              (:then_0
                (CallFunction|CheckRedRiftCalendarEvents))
              (:then_1
                (CallFunction|CheckNextMatchCalendarEvents)))))))))

(event Custom|Check%20Next%20Match%20Calendar%20Events
  ; RudeLemon / Rocket Landing
  ;;[0.6823 0.6186 0.206 0.5] Next Match Calendar Events
  (Default|LogIt "RudeLemon - check next match calendar events")
  (bind _biseventactive (CalendarEvents|GetCalendarEventInformation "RL_PostEvent"))
  (if _biseventactive
    (CallFunction|RLNextMatch 1)))

(event Custom|Check%20Red%20Rift%20Calendar%20Events
  ; RudeLemon / Rocket Landing
  ;;[0.7396 0.2699 0.3266 0.5] Red Rift Calendar Events
  (Default|LogIt "check red rift calendar events")
  (Utilities|FlowControl|Sequence
    (:then_0
      (Default|LogIt "red rift Day 1 calendar active")
      (bind _biseventactive (CalendarEvents|GetCalendarEventInformation "RL_Rift1"))
      (if _biseventactive
        (CallFunction|RedRiftDay 1) ; red rift day 1
      ))
    (:then_1
      (Default|LogIt "red rift Day 2 calendar active")
      (bind _biseventactive_2 (CalendarEvents|GetCalendarEventInformation "RL_Rift2"))
      (if _biseventactive_2
        (CallFunction|RedRiftDay 2) ; red rift day 2
      ))) ; 0 - check day 1\n1 - check day 2
)

(event Custom|PostRudeLemon
  ; RudeLemon / Rocket Landing
  ;;[0.6823 0.6186 0.206 0.5] Next match data layers
  (CallFunction|RLNextMatch 1))

(event Custom|RL_NextMatch
  (params (next%20match%20state :type Integer))
  ; RudeLemon / Rocket Landing
  ;;[0.6823 0.6186 0.206 0.5] Next match data layers
  (switch int next%20match%20state
    (:Default
      _)
    (:0
      _)
    (:1
      (bind _returnvalue (WorldPartition|GetDataLayerManager)) ; 1 - set to "next match" state
      (DataLayers|SetDataLayerRuntimeState _returnvalue "/DragonCartNarrative/DataLayers/RL_PostLanding.RL_PostLanding" "Activated" false :!class /Script/Engine.DataLayerManager) ; next match data layer - Activated
      (DataLayers|SetDataLayerRuntimeState _returnvalue "/DragonCartNarrative/DataLayers/RL_ThrusterPit_Loot.RL_ThrusterPit_Loot" "Unloaded" false :!class /Script/Engine.DataLayerManager) ; thruster pit loot - data layer - Unloaded
      (DataLayers|SetDataLayerRuntimeState _returnvalue "/DragonCartNarrative/DataLayers/RL_LandingDebris.RL_LandingDebris" "Activated" false :!class /Script/Engine.DataLayerManager) ; lower tunnel debris - data layer - Activated
      (DataLayers|SetDataLayerRuntimeState _returnvalue "/DragonCartNarrative/DataLayers/RL_Pre.RL_Pre" "Unloaded" false :!class /Script/Engine.DataLayerManager) ; RL_Pre - Unloaded
    ))
  (bind _returnvalue_2 (WorldPartition|GetDataLayerManager)) ; default - reset to pre event state
  (DataLayers|SetDataLayerRuntimeState _returnvalue_2 "/DragonCartNarrative/DataLayers/RL_PostLanding.RL_PostLanding" "Unloaded" false :!class /Script/Engine.DataLayerManager) ; next match data layer - Unloaded
  (DataLayers|SetDataLayerRuntimeState _returnvalue_2 "/DragonCartNarrative/DataLayers/RL_ThrusterPit_Loot.RL_ThrusterPit_Loot" "Activated" false :!class /Script/Engine.DataLayerManager) ; thruster pit loot - data layer - Activated
  (DataLayers|SetDataLayerRuntimeState _returnvalue_2 "/DragonCartNarrative/DataLayers/RL_LandingDebris.RL_LandingDebris" "Unloaded" false :!class /Script/Engine.DataLayerManager) ; lower tunnel debris - data layer - Unloaded
  (DataLayers|SetDataLayerRuntimeState _returnvalue_2 "/DragonCartNarrative/DataLayers/RL_Pre.RL_Pre" "Activated" false :!class /Script/Engine.DataLayerManager) ; RL_Pre - Activated
)

(event Custom|RedRiftDay
  (params (red%20rift%20stage :type Integer))
  ; RudeLemon / Rocket Landing
  ;;[0.7396 0.2699 0.3266 0.5] Red Rift - data layers
  (switch int red%20rift%20stage
    (:Default
      _)
    (:0
      _)
    (:1
      (bind _returnvalue (WorldPartition|GetDataLayerManager)) ; Day 1
      (DataLayers|SetDataLayerRuntimeState _returnvalue "/DragonCartNarrative/DataLayers/DL_RedRift_Day1.DL_RedRift_Day1" "Activated" false :!class /Script/Engine.DataLayerManager) ; Day 1 Red Rift - Activated
      (DataLayers|SetDataLayerRuntimeState _returnvalue "/DragonCartNarrative/DataLayers/DL_RedRift_Day2.DL_RedRift_Day2" "Unloaded" false :!class /Script/Engine.DataLayerManager) ; Day 2 Red Rift - Disabled
    )
    (:2
      (bind _returnvalue_2 (WorldPartition|GetDataLayerManager)) ; Day 2
      (DataLayers|SetDataLayerRuntimeState _returnvalue_2 "/DragonCartNarrative/DataLayers/DL_RedRift_Day2.DL_RedRift_Day2" "Activated" false :!class /Script/Engine.DataLayerManager) ; Day 2 Red Rift - Activated
      (DataLayers|SetDataLayerRuntimeState _returnvalue_2 "/DragonCartNarrative/DataLayers/DL_RedRift_Day1.DL_RedRift_Day1" "Unloaded" false :!class /Script/Engine.DataLayerManager) ; Day 1 Red Rift - Disabled
    ))
  (bind _returnvalue_3 (WorldPartition|GetDataLayerManager)) ; default - disable all
  (DataLayers|SetDataLayerRuntimeState _returnvalue_3 "/DragonCartNarrative/DataLayers/DL_RedRift_Day1.DL_RedRift_Day1" "Unloaded" false :!class /Script/Engine.DataLayerManager) ; Day 1 Red Rift - Disabled
  (DataLayers|SetDataLayerRuntimeState _returnvalue_3 "/DragonCartNarrative/DataLayers/DL_RedRift_Day2.DL_RedRift_Day2" "Unloaded" false :!class /Script/Engine.DataLayerManager) ; Day 2 Red Rift - Disabled
)

(fn CanUpdateInMatch
  (if (|CallParentFunction)
    (Class|BuildingFunctionLibrary|LogString "NewEnumerator5" "Attempting to update a Data Layer Calendar Event during a match.  This is not supported due to performance concerns." true self)
    (return false)))

(fn GetEventNames
  (for _array_element (Variables|Default|GetCalendarDataLayerList)
    (if (not (Utilities|String|IsEmpty (Utilities|Struct|BreakCalendarDataLayerData _array_element)))
      (Utilities|Array|AddUnique (LocalVariables|GetLocalEventNames) (Utilities|Struct|BreakCalendarDataLayerData _array_element))))
  (return (LocalVariables|GetLocalEventNames)))

(fn HandleEventActive
  (params (EventName :type String))
  (CallFunction|HandleEventStateChanged EventName true))

(fn HandleEventEnded
  (params (EventName :type String))
  (CallFunction|HandleEventStateChanged EventName))

(fn HandleEventStateChanged
  (params (EventName :type String) (bIsEventActive :type Boolean))
  (bind _biseventactive (Variables|GetbIsEventActive))
  (bind _returnvalue (select _biseventactive "Unloaded" "Activated"))
  (bind _biseventactive_2 (Variables|GetbIsEventActive))
  (bind _returnvalue_2 (select _biseventactive_2 "Activated" "Unloaded"))
  (bind _eventname (Variables|GetEventName))
  (for _array_element (Variables|Default|GetCalendarDataLayerList)
    (if (Utilities|String|EqualExactly(String) (Variables|GetEventName) (Utilities|Struct|BreakCalendarDataLayerData _array_element))
      (CallFunction|LogIt (Utilities|String|ToString(Text) (Utilities|Text|FormatText "NSLOCTEXT(\"[1BFFABD9577B579936C513064BBE2265]\", \"609430AB4727D3606A6C7B8C178DD394\", \"{a} : BP_CalendarDataLayer:HandleEventActive - Calendar Event Name: {b} was found.\")" (Variables|Default|GetDisplayName) _eventname)))
      (bind _returnvalue_3 (WorldPartition|GetDataLayerManager))
      (CallFunction|LogIt (Utilities|String|ToString(Text) (Utilities|Text|FormatText "NSLOCTEXT(\"[1BFFABD9577B579936C513064BBE2265]\", \"3BBB477E4EC6BF53DE7CE8A85F69221A\", \"{a} : BP_CalendarDataLayer:HandleEventActive - Getting Data Layer Manager : {b}\")" (Variables|Default|GetDisplayName) _returnvalue_3)))
      (Utilities|FlowControl|Sequence
        (:then_0
          (for _array_element_2 (.EnabledDataLayers_9_41BB9A3E4648805B5AF70AA728288C92 (Utilities|Struct|BreakCalendarDataLayerData _array_element))
            (DataLayers|SetDataLayerInstanceRuntimeState _returnvalue_3 (DataLayers|GetDataLayerInstancefromAsset _returnvalue_3 _array_element_2) _returnvalue_2 false)
            (CallFunction|LogIt (Utilities|String|ToString(Text) (Utilities|Text|FormatText "NSLOCTEXT(\"[1BFFABD9577B579936C513064BBE2265]\", \"95DFCD6F426D018E3B28FD9D75F6902C\", \"{a} : BP_CalendarDataLayer:HandleEventActive:ENABLED Layers  from the Enable Layer- Calling Set Data Layer Instance Runtime State with the following information:\\r\\nDataLayer Manager: {b}\\r\\nDataLayer Instance: {c}\\r\\nSetting to State: {d}\\r\\nBecause Event is: {e}\")" (Variables|Default|GetDisplayName) _returnvalue_3 (DataLayers|GetDataLayerInstancefromAsset _returnvalue_3 _array_element_2) (Utilities|String|EnumtoString _returnvalue_2) (Variables|GetbIsEventActive))))))
        (:then_1
          (for _array_element_3 (.DisabledDataLayers_10_8DE59E3B474C88ED2E7988B185572444 (Utilities|Struct|BreakCalendarDataLayerData _array_element))
            (DataLayers|SetDataLayerInstanceRuntimeState _returnvalue_3 (DataLayers|GetDataLayerInstancefromAsset _returnvalue_3 _array_element_3) _returnvalue false)
            (CallFunction|LogIt (Utilities|String|ToString(Text) (Utilities|Text|FormatText "NSLOCTEXT(\"[1BFFABD9577B579936C513064BBE2265]\", \"349D04DD4521A3570B79CC9E7B65A2F8\", \"{a} : BP_CalendarDataLayer:HandleEventActive:DISABLED Layers from the Disable Array - Calling Set Data Layer Instance Runtime State with the following information:\\r\\nDataLayer Manager: {b}\\r\\nDataLayer Instance: {c}\\r\\nSetting to State: {d}\\r\\nBecause Event is: {e}\")" (Variables|Default|GetDisplayName) _returnvalue_3 (DataLayers|GetDataLayerInstancefromAsset _returnvalue_3 _array_element_3) (Utilities|String|EnumtoString _returnvalue) (Variables|GetbIsEventActive)))))))
      (else
        (CallFunction|LogIt (Utilities|String|ToString(Text) (Utilities|Text|FormatText "NSLOCTEXT(\"[1BFFABD9577B579936C513064BBE2265]\", \"2F82341B45E9C8224CCBD4855EE3DEFE\", \"{a} : BP_CalendarDataLayer:HandleEventActive - Calendar Event Name: {b} was NOT found.\")" (Variables|Default|GetDisplayName) _eventname)) true)))))

(fn LogIt
  (params (InString :type String) (bLogAsWarning :type Boolean))
  ; Log It
  (Fortnite|FortShippingLog 0 (Utilities|String|Append "[BP_RudeLemon_PostCalendar] - " InString) bLogAsWarning))

(fn ConstructionScript
  (|CallParentFunction))
