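; BP_ShockKnockback is a Blueprint knockback object that adds a gameplay effect to player pawns it impulses.
; It overrides OnPlayerImpulsed, calls the parent function, then applies its configurable ImpulseGameplayEffect through the pawn's Fort ability system component, which is both source and target.
; The impulse itself is not defined here; BeginPlay, overlap and Tick events are empty or disabled stubs.
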(event EventBeginPlay
  (|CallParentFunction) ; This node is disabled and will not be called.\nDrag off pins to build functionality.
)

(event Collision|EventActorBeginOverlap
  (params (OtherActor :type Actor%20Object%20Reference)))

(event EventTick
  (params (DeltaSeconds :type Float%20%28single-precision%29)))

(fn OnPlayerImpulsed
  (params (ImpulsedPawn :type Fort%20Player%20Pawn%20Object%20Reference))
  (|CallParentFunction ImpulsedPawn)
  (GameplayEffects|ApplyGameplayEffectToTarget :self (Ability|GetFortAbilitySystemComponent ImpulsedPawn) :GameplayEffectClass (Variables|Default|GetImpulseGameplayEffect) :Target (Ability|GetFortAbilitySystemComponent ImpulsedPawn) :Level 0.0))

(fn ConstructionScript
  (|CallParentFunction))
